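# Translucent, unlit, two-sided Niagara sprite and mesh particle material for the Pulse Lancer (Masamune) heavy attack's big slash energy.
# Emissive adds panning noise and caustics, UV-distorted by blurred noise and tinted by particle color.
# Opacity combines panning noise, sine edge falloffs, particle alpha, depth fade and vertex-color-driven view-angle fade, then a dynamic-parameter segmented mask.
# A dynamic random offset shifts the shared tiled UVs.
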Material /Game/Effects/Fort_Effects/GameplayPulginsAssets/PulseLancer/HeavyAttack/M_Masamune_BigSlashEnergy
graph {
 properties {
  bAutomaticallySetUsageInEditor=false
  bUsedWithNiagaraMeshParticles=true
  bUsedWithNiagaraSprites=true
  blendMode=BLEND_Translucent
  shadingModel=MSM_Unlit
  twoSided=true
  usedShadingModels=Unlit
 }
 nodes [
  n1 = ParticleColor()
  n2 = NamedRerouteUsage()
  n3 = Multiply(A=TextureSample(UVs=Panner(Time=@n2) {speedY=0.1}) {samplerType=SAMPLERTYPE_Grayscale, texture=&LowResBlurredNoise}.R) {constB=0.2}
  n4 = Add(A=NamedRerouteUsage(), B=@n3)
  n5 = NamedRerouteUsage()
  n6 = Desaturation(
   None=TextureSampleParameter2D(UVs=Panner(Coordinate=@n4, Time=@n5) {speedX=-0.5}) {parameterName=Texture Noise, texture=/Game/Effects/Fort_Effects/Textures/Tile/T_Noise_03}.G
   Fraction=ScalarParameter() {defaultValue=-3}
  )
  n7 = NamedRerouteUsage()
  n8 = TextureSampleParameter2D(UVs=Panner(Coordinate=Add(A=@n7, B=@n3), Time=NamedRerouteUsage()) {speedX=-0.3}) {
   parameterName=Texture Caustics
   samplerType=SAMPLERTYPE_LinearGrayscale
   texture=/Game/Global/Textures/Default/Caustics/T_Nosie_Caustics_Grayscale
  }
  n9 = Power(Base=@n8.R) {constExponent=1}
  MaterialFunctionCall(Emissive=Multiply(A=@n1.RGB, B=Add(A=@n6, B=@n9))) {materialFunction=&MF_FX_EmissiveAdjustments}
  n10 = Power(Base=DepthFade(FadeDistance=ScalarParameter() {defaultValue=50, parameterName=DF}))
  n11 = NamedRerouteUsage()
  n12 = Multiply(A=TextureSample() {samplerType=SAMPLERTYPE_Grayscale, texture=&LowResBlurredNoise}.R) {constB=0.2}
  n13 = TextureSampleParameter2D(UVs=Panner(Coordinate=Add(A=@n11, B=@n12), Time=NamedRerouteUsage()) {speedX=0.5}) {parameterName=Texture, samplerType=SAMPLERTYPE_Grayscale, texture=&LowResBlurredNoise}
  n14 = Sine(None=ComponentMask(None=TextureCoordinate()) {r=true}) {period=2}
  n15 = Sine(None=ComponentMask(None=TextureCoordinate()) {g=true}) {period=2}
  n16 = Multiply(A=@n1.A, B=Multiply(A=Multiply(A=@n13.R, B=@n14) {constA=1}, B=Power(Base=@n15)))
  n17 = ScalarParameter() {defaultValue=1, group=LookDev, parameterName=Texture Power}
  n18 = LinearInterpolate(Alpha=VertexColor().R) {constB=5}
  n19 = Saturate(None=Multiply(
    A=Power(Base=@n16, Exp=@n17)
    B=MaterialFunctionCall("FadeSensitivity (optional)"=@n18) {materialFunction=&FadeBasedOnViewAngle}.Result
   ))
  n20 = DynamicParameter() {parameterIndex=3, defaultValue=<rgba 0 0 0 0>, paramNames [InputMask, Random Offset, Param3, Param4]}
  Multiply(
   A=Saturate(None=Multiply(A=@n10, B=@n19))
   B=MaterialFunctionCall!(InUV, MaskInput=@n20.InputMask) {materialFunction=&/MF_Segmented}.Result
  )
  n21 = TextureCoordinate()
  n22 = DynamicParameter() {parameterIndex=3, defaultValue=<rgba 0 0 0 0>, paramNames [InputMask, Random Offset, Param3, Param4]}
  n23 = ScalarParameter() {defaultValue=1, group=Tiling, parameterName=UV Tile X}
  n24 = ScalarParameter() {defaultValue=1, group=Tiling, parameterName=UV Tile Y}
  NamedRerouteDeclaration(None=Multiply(A=Add(A=@n21, B=@n22.Random Offset), B=AppendVector(A=@n23, B=@n24))) {name=UV, nodeColor=<rgba 0.5703 1 0 1>}
  NamedRerouteDeclaration(None=Multiply(A=Time(), B=ScalarParameter() {defaultValue=1, parameterName=Panning Speed})) {name=Time, nodeColor=<rgba 0.1484 0 1 1>}
 ]
}
