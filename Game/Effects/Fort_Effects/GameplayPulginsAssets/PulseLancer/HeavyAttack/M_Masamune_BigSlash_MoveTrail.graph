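# Additive, unlit, two-sided Niagara mesh particle material for the moving trail of the Masamune heavy-attack big slash in the PulseLancer effects.
# Leftward and rightward edge-faded UV masks shape scrolling noise, flame and trail textures, with blurred noise distorting the latter two.
# Particle color tints emissive, hue-shifted along U.
# Opacity combines flame, edge masks, particle alpha, depth fade and a segmented mask driven by a dynamic parameter.
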Material /Game/Effects/Fort_Effects/GameplayPulginsAssets/PulseLancer/HeavyAttack/M_Masamune_BigSlash_MoveTrail
graph {
 properties {
  bAutomaticallySetUsageInEditor=false
  bUsedWithNiagaraMeshParticles=true
  blendMode=BLEND_Additive
  shadingModel=MSM_Unlit
  twoSided=true
  usedShadingModels=Unlit
 }
 nodes [
  CurveAtlasRowParameter(CurveTime=ComponentMask(None=TextureCoordinate()) {r=true}) {
   atlas=/Game/Effects/Fort_Effects/GameplayPulginsAssets/PulseLancer/Slash/Material/CAtlas_Masamune_Color
   curve=/Game/Effects/Fort_Effects/GameplayPulginsAssets/PulseLancer/Slash/Material/CV_Masamune_SlashColor4
   defaultValue=4
   parameterName=Color
  }
  n1 = Add(A=TextureCoordinate() {uTiling=0.4}, B=Constant2Vector() {g=0.3})
  n2 = Multiply(A=TextureSample(UVs=Panner(Coordinate=@n1) {speedX=-0.5}) {texture=/Game/Effects/Fort_Effects/Textures/Tile/T_Noise_03}.G) {constB=2}
  Desaturation(None=@n2, Fraction=Constant())
  n3 = MaterialFunctionCall(Float2=TextureCoordinate()) {materialFunction=&BreakOutFloat2Components}
  n4 = Multiply(
   A=Add(A=@n3.R, B=ScalarParameter() {group=HS, parameterName=HS_GrayAdjust_Add})
   B=ScalarParameter() {defaultValue=1, group=HS, parameterName=HS_GrayAdjust_Mul}
  )
  n5 = Saturate(None=Power(Base=Saturate(None=@n4), Exp=ScalarParameter() {defaultValue=2.0097, group=HS, parameterName=HS_GrayAdjust_Power}))
  n6 = ScalarParameter() {defaultValue=0.1, parameterName=HHHS}
  n7 = ParticleColor()
  n8 = VectorParameter() {parameterName=ColorNoise, defaultValue=<rgba 2 1.5977 1.1768 1>}
  n9 = TextureCoordinate() {uTiling=0.3, vTiling=2}
  n10 = TextureCoordinate() {uTiling=0.5}
  n11 = Multiply(A=TextureSample(UVs=Panner(Coordinate=@n10) {speedX=-0.1, speedY=0.1}) {samplerType=SAMPLERTYPE_Grayscale, texture=&LowResBlurredNoise}.R) {constB=0.2}
  n12 = TextureSample(UVs=Panner(Coordinate=Add(A=@n9, B=@n11)) {speedX=-1}) {samplerType=SAMPLERTYPE_LinearGrayscale, texture=/Game/Weapons/FORT_Melee/Pickaxe_Cube_Ninja_Male/FX/T_FluidFlameTile}
  n13 = Power(Base=@n12.R) {constExponent=5}
  n14 = TextureCoordinate() {uTiling=0.3}
  n15 = TextureSampleParameter2D(UVs=Panner(Coordinate=Add(A=@n14, B=@n11)) {speedX=-1}) {parameterName=Default_Tex, samplerType=SAMPLERTYPE_Masks, texture=/Game/Effects/Fort_Effects/GameplayPulginsAssets/Masamune/T_Masamune_Trail}
  n16 = Multiply(A=@n15.G)
  n17 = Add(A=@n2, B=Add(A=Multiply(A=@n8.RGB, B=@n13), B=Multiply(B=@n16) {constA=1}))
  MaterialFunctionCall(Emissive=MaterialFunctionCall(Hue Shift Percentage=Multiply(A=@n5, B=@n6), Texture=Multiply(A=@n7.RGB, B=@n17)) {materialFunction=&HueShift}.Result) {materialFunction=&MF_FX_EmissiveAdjustments}
  n18 = ScalarParameter() {defaultValue=0.25, parameterName=Bas}
  n19 = ComponentMask(None=TextureCoordinate()) {r=true}
  n20 = Power(Base=ComponentMask(None=TextureCoordinate()) {r=true})
  n21 = Multiply(A=Power(Base=OneMinus(None=@n19)), B=Saturate(None=Multiply(A=@n20) {constB=5}))
  n22 = ComponentMask(None=TextureCoordinate()) {g=true}
  n23 = Saturate(None=Multiply(A=Sine(None=@n22) {period=2}) {constB=2})
  n24 = Multiply(A=ParticleColor().A, B=Multiply(A=Add(A=@n12.R, B=@n18) {constB=0.75}, B=Multiply(A=@n21, B=@n23)) {constA=1})
  n25 = ScalarParameter() {defaultValue=100, parameterName=FadeDistance}
  n26 = TextureCoordinate()
  Multiply(
   A=Saturate(None=DepthFade(Opacity=@n24, FadeDistance=@n25))
   B=MaterialFunctionCall!(
    InUV=MaterialFunctionCall(XY=@n26) {materialFunction=&Swizzle}.YX
    MaskInput=DynamicParameter() {parameterIndex=3, defaultValue=<rgba 0 0.2 0 0>, paramNames [InputMask,Param2,Param3,Param4]}.InputMask
   ) {materialFunction=&/MF_Segmented}.Result
  )
  OneMinus(None=@n3.R)
 ]
}
