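# Unlit, additive, two-sided material for Niagara mesh particles, a simple glow for the Pulse Lancer Masamune big-slash heavy attack effect.
# Sine waves on the U and V coordinates multiply into a procedural soft gradient, tinted by particle color and passed through an emissive-adjustment function.
# Particle alpha times that gradient feeds a depth fade, multiplied by an MF_Segmented mask that takes a dynamic parameter as input.
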Material /Game/Effects/Fort_Effects/GameplayPulginsAssets/PulseLancer/HeavyAttack/M_Masamune_BigSlash_SimpleGlow
graph {
 properties {
  bAutomaticallySetUsageInEditor=false
  bUsedWithNiagaraMeshParticles=true
  blendMode=BLEND_Additive
  shadingModel=MSM_Unlit
  twoSided=true
  usedShadingModels=Unlit
 }
 nodes [
  LinearInterpolate(Alpha=VertexColor().R) {constB=10}
  n1 = ParticleColor()
  n2 = TextureCoordinate()
  n3 = Sine(None=ComponentMask(None=@n2) {r=true}) {period=2}
  n4 = Multiply(A=@n3, B=Sine(None=ComponentMask(None=@n2) {g=true}) {period=2})
  MaterialFunctionCall(Emissive=Multiply(A=@n1.RGB, B=@n4)) {materialFunction=&MF_FX_EmissiveAdjustments}
  MaterialFunctionCall() {materialFunction=&FadeBasedOnViewAngle}
  n5 = Multiply(A=@n1.A, B=@n4)
  n6 = ScalarParameter() {defaultValue=100, parameterName=FadeDistance}
  Multiply(
   A=Saturate(None=DepthFade(Opacity=@n5, FadeDistance=@n6))
   B=MaterialFunctionCall!(InUV, MaskInput=DynamicParameter() {parameterIndex=3, defaultValue=<rgba 0 0.2 0 0>, paramNames [InputMask,Param2,Param3,Param4]}.InputMask) {materialFunction=&/MF_Segmented}.Result
  )
  Multiply(A=@n5)
  Saturate(None=Multiply(A=@n3) {constB=2})
 ]
}
