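# Translucent, unlit, two-sided Niagara mesh-particle material for the small trail of the PulseLancer Masamune heavy-attack big slash.
# Panning soft noise distorts UVs for two blended scrolling trail textures, which combine with a V gradient and particle color.
# An inverted edge mask drives a subtle hue shift on that color.
# Opacity math multiplies particle alpha, trail, edge and sine masks, squared depth fade and a segmented mask driven by a dynamic parameter.
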Material /Game/Effects/Fort_Effects/GameplayPulginsAssets/PulseLancer/HeavyAttack/M_Masamune_BigSlash_SmallTrail
graph {
 properties {
  bAutomaticallySetUsageInEditor=false
  bUsedWithNiagaraMeshParticles=true
  blendMode=BLEND_Translucent
  shadingModel=MSM_Unlit
  twoSided=true
  usedShadingModels=Unlit
 }
 nodes [
  Constant() {r=1}
  n1 = TextureCoordinate()
  n2 = VectorParameter() {group=HueS, parameterName=HueS_EdgeMask_Control, channelNames {a="", b=Power, g=Multiply, r=Offset}, defaultValue=<rgba 0 2 1.5 1>}
  n3 = OneMinus(None=MaterialFunctionCall!(
    UorV=MaterialFunctionCall(Float2=@n1) {materialFunction=&BreakOutFloat2Components}.G
    EdgeControl=@n2
   ) {materialFunction=/Game/Effects/Niagara/MasterMaterials/Functions/Common/FXCommon_EdgeMask_Simple}.Mask)
  n4 = ScalarParameter() {defaultValue=0.05, parameterName=HS}
  n5 = ParticleColor()
  n6 = ComponentMask(None=TextureCoordinate()) {g=true}
  n7 = Panner(Coordinate=TextureCoordinate() {uTiling=2}) {speedY=2}
  n8 = Multiply(A=ComponentMask(None=TextureSample(UVs=@n7) {texture=/Game/Effects/Fort_Effects/Emotes/UltraLight/Texture/T_SoftNoise_BC7}.RGB) {g=true, r=true}) {constB=0.2}
  n9 = Add(A=TextureCoordinate(), B=@n8)
  n10 = Add(A=TextureCoordinate(), B=@n8)
  n11 = LinearInterpolate(
   A=TextureSample(UVs=Panner(Coordinate=@n9) {speedX=0.2}) {samplerType=SAMPLERTYPE_Grayscale, texture=&T_Smoke_Trail_Soft}.R
   B=TextureSample(UVs=Panner(Coordinate=@n10) {speedX=0.2}) {samplerType=SAMPLERTYPE_LinearColor, texture=/Game/Effects/Fort_Effects/Textures/T_CubeNinja_Trail}.R
  ) {constAlpha=0.3}
  n12 = Add(A=Power(Base=@n6), B=@n11)
  MaterialFunctionCall(Emissive=MaterialFunctionCall(Hue Shift Percentage=Multiply(A=@n3, B=@n4), Texture=Multiply(A=@n5.RGB, B=@n12)) {materialFunction=&HueShift}.Result) {materialFunction=&MF_FX_EmissiveAdjustments}
  MaterialFunctionCall() {materialFunction=&FadeBasedOnViewAngle}
  n13 = Multiply(A=@n5.A, B=Add(A=Power(Base=@n11) {constExponent=1.5}) {constB=0})
  n14 = ComponentMask(None=TextureCoordinate()) {r=true}
  n15 = VectorParameter() {group=EdgeF, parameterName=EdgeF_EdgeMask_Control, channelNames {a="", b=Power, g=Multiply, r=Offset}, defaultValue=<rgba 0 2 4.4456 1>}
  n16 = TextureCoordinate()
  n17 = Multiply(
   A=MaterialFunctionCall(UorV=@n14, EdgeControl=@n15) {materialFunction=&FXCommon_EdgeMask_Simple}.Mask
   B=Sine(None=ComponentMask(None=@n16) {g=true}) {period=2}
  )
  n18 = Saturate(None=Multiply(A=Multiply(A=@n13, B=@n17)))
  n19 = DepthFade(FadeDistance=ScalarParameter() {defaultValue=100, parameterName=FadeDistance})
  n20 = Multiply(A=@n19, B=@n19)
  Multiply(
   A=Saturate(None=Multiply(A=@n18, B=@n20))
   B=MaterialFunctionCall!(InUV, MaskInput=DynamicParameter() {parameterIndex=3, defaultValue=<rgba 0 0.2 0 0>, paramNames [InputMask,Param2,Param3,Param4]}.InputMask) {materialFunction=&/MF_Segmented}.Result
  )
  n21 = Panner(Coordinate=TextureCoordinate() {uTiling=2}) {speedY=2}
  Multiply(A=ComponentMask(None=TextureSample(UVs=@n21) {texture=&T_SoftNoise_BC7}.RGB) {g=true, r=true}) {constB=0.15}
  Power(Base=@n6) {constExponent=7}
  Sine(None=@n14) {period=2}
 ]
}
